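; Season 28 prelude quest ability that has a chance to drop a colored tile quest actor when the avatar deals weakspot damage.
; Targets with excluded tags spawn nothing, and a brief CanSpawn lockout throttles drops.
; CheckTileColor picks red, purple or blue from the player's q03 and q04 quest instances.
; SpawnQuestActorsFromDropEntry launches the tile from a transform derived from the hit target and optionally copies its tags.
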(event Ability|EventActivateAbility
  (bind _returnvalue (Ability|GetAvatarActorfromActorInfo))
  (bind _returnvalue_2 (Math|Random|RandomFloatinRange 0.0 1.0))
  (Ability|Async|WaitGameplayEventToActor _returnvalue "(TagName=\"Event.SkydiveEnded.FromBus\")" true
    (:then
      (if (Variables|Default|GetHasLanded)
        _))
    (:EventReceived
      (Variables|Default|SetHasLanded true)))
  (bind (_asynctaskproxy _payload) (Ability|Async|WaitGameplayEventToActor _returnvalue "(TagName=\"Event.Damage.Dealt.Weakspot\")" false false)
    (:EventReceived
      (if (Variables|Default|GetCanSpawn)
        (bind (_payload_eventtag _payload_instigator _payload_target _payload_optionalobject _payload_optionalobject2 _payload_contexthandle _payload_instigatortags _payload_targettags _payload_eventmagnitude _payload_targetdata) (Utilities|Struct|BreakGameplayEventData _payload))
        (if (GameplayTags|HasAnyTags _payload_targettags (Variables|Default|GetTag%20Container))
          (Development|PrintString (Utilities|String|ToString(Float) _returnvalue_2) true true "(R=1.000000,G=0.000000,B=0.421799,A=1.000000)")
          (if (<= _returnvalue_2 (Variables|Default|GetTileDropChance))
            (bind _color (Default|CheckTileColor))
            (switch string _color
              (:Default
                (Spawning|SpawnQuestActorsFromDropEntry 0 (Utilities|Struct|MakeStructSpawnedQuestActorDropEntry (Variables|Default|GetBlueTile) 1 1) (Utilities|Struct|MakeGameplayEventData _payload_eventtag _payload_instigator _payload_target _payload_optionalobject _payload_optionalobject2 _payload_contexthandle _payload_instigatortags _payload_targettags _payload_eventmagnitude _payload_targetdata) :!label _label))
              (:Red
                (Spawning|SpawnQuestActorsFromDropEntry 0 (Utilities|Struct|MakeStructSpawnedQuestActorDropEntry (Variables|Default|GetRedTile) 1 1) (Utilities|Struct|MakeGameplayEventData _payload_eventtag _payload_instigator _payload_target _payload_optionalobject _payload_optionalobject2 _payload_contexthandle _payload_instigatortags _payload_targettags _payload_eventmagnitude _payload_targetdata)))
              (:Purple
                (Spawning|SpawnQuestActorsFromDropEntry 0 (Utilities|Struct|MakeStructSpawnedQuestActorDropEntry (Variables|Default|GetPurpleTile) 1 1) (Utilities|Struct|MakeGameplayEventData _payload_eventtag _payload_instigator _payload_target _payload_optionalobject _payload_optionalobject2 _payload_contexthandle _payload_instigatortags _payload_targettags _payload_eventmagnitude _payload_targetdata)))
              (:Blue
                (goto _label)))
            (Variables|Default|SetCanSpawn false)
            (Utilities|FlowControl|Delay 0.2)
            (Variables|Default|SetCanSpawn true)))))))

(fn CheckTileColor
  (bind _returnvalue (Ability|GetAvatarActorfromActorInfo))
  (bind _returnvalue_2 (Quest|GetQuestComponentforPlayerObject _returnvalue))
  (bind _returnvalue_3 (Quest|FindQuestInstanceforParticipant _returnvalue_2 "Quest:quest_s28_prelude_q04"))
  (Utilities|Casting|CastToFortQuestItem _returnvalue_3
    (:then
      (bind _returnvalue_4 (Quest|GetQuestComponentforPlayerObject _returnvalue))
      (bind _returnvalue_5 (Quest|FindQuestInstanceforParticipant _returnvalue_4 "Quest:quest_s28_prelude_q03"))
      (Utilities|Casting|CastToFortQuestItem _returnvalue_5
        (:then
          (return "Blue"))
        (:CastFailed
          (return "Purple"))))
    (:CastFailed
      (return "Red"))))

(fn SpawnQuestActorsFromDropEntry
  (params (DropEntry :type Struct%20Spawned%20Quest%20Actor%20Drop%20Entry%20Structure) (EventData :type Gameplay%20Event%20Data%20Structure))
  (bind (_spawnedquestactorclass_6_f15cfde2424412e9badbe2975b317a07 _dropquantity_14_6469e6844f83772bb80669bcf263cb9d) (Utilities|Struct|BreakStructSpawnedQuestActorDropEntry DropEntry))
  (bind (_eventtag _instigator _target _optionalobject _optionalobject2 _contexthandle _instigatortags _targettags _eventmagnitude _targetdata) (Utilities|Struct|BreakGameplayEventData EventData))
  (bind (_eventtag_2 _instigator_2 _target_2 _optionalobject_2) (Utilities|Struct|BreakGameplayEventData EventData))
  (for _ (range 1 (+ _dropquantity_14_6469e6844f83772bb80669bcf263cb9d 1))
    (bind _returnvalue (Building|SpawnBuildingGameplayActor _spawnedquestactorclass_6_f15cfde2424412e9badbe2975b317a07 (Spawning|GetSpawnTransform _target _targetdata) _instigator))
    (bind _asbp_spawned_quest_actor_parent (Utilities|Casting|CastToBP_SpawnedQuestActor_Parent _returnvalue))
    (Variables|/Game/Athena/Items/QuestParents/SpawnedQuestActor/BP_SpawnedQuestActor_Parent.BP_SpawnedQuestActor_Parent_C|SetbTestVisualsNoQuestRequired false _asbp_spawned_quest_actor_parent)
    (Projectile|SetVelocity (Variables|/Game/Athena/Items/QuestParents/SpawnedQuestActor/BP_SpawnedQuestActor_Parent.BP_SpawnedQuestActor_Parent_C|GetFortProjectileMovement _asbp_spawned_quest_actor_parent) (Spawning|GenerateSpawnVelocity))
    (Class|BPSpawnedQuestActorParent|CheckCosmetics _asbp_spawned_quest_actor_parent (Fortnite|GetFortPlayerControllerfromActor (Variables|GAVars|Data|GetOwningPlayerPawn)))
    (if (Variables|Setup|GetSpawnedActorInheritsTags)
      (Ability|GameplayEffect|AddLooseGameplayTags _asbp_spawned_quest_actor_parent _targettags))
    (Utilities|IsValid _optionalobject_2
      (:Is%20Valid
        (bind _asactor (Utilities|Casting|CastToActor _optionalobject_2))))))
